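; Audio ID component whose Assign Weapon ID event gives a weapon instance a random integer ID in a small fixed range.
; On non-server machines, it then prints the owner's name and assigned ID, with screen and log output controlled by DebugEnabled.
; A note marks this as temporary, since IDs can repeat, and proposes incrementing per spawned weapon in version 2.0.
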(event Custom|Assign%20Weapon%20ID
  ;[0.1455 0.3094 0] This is a temporary quick/easy way to assign IDs to each instance of a weapon. Version 2.0 of this component should increment Weapon ID for each instance of the weapon that is spawned at runtime to minimize duplicate IDs.
  (Variables|Default|SetWeaponID (Math|Random|RandomIntegerinRange 0 4))
  (if (not (Networking|IsServer))
    (Development|PrintString (Utilities|String|Append (Utilities|GetDisplayName (Components|GetOwner)) " recieved Weapon ID: " (Utilities|String|ToString(Integer) (Variables|Default|GetWeaponID))) (Variables|Default|GetDebugEnabled) (Variables|Default|GetDebugEnabled))))
